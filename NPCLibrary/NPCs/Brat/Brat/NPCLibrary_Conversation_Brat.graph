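# Conversation database for The Brat NPC, providing its greeting and the top-level branches of its dialogue.
# The Brat entry point leads to one Speech node whose default line is swapped by contextual messages for first-ever meetings, match conversation counts or specific player character cosmetics.
# That speech fans out through link nodes to remote entry tags for NPC Library quests, basic services, Brat dialogue, and Season 36 Big Dill and DJ Yonder character stories.
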ConversationDatabase /NPCLibrary/NPCs/Brat/Brat/NPCLibrary_Conversation_Brat
entry_points [{tag=Conversation.Entry.Brat, destinations [#1]}]
nodes [
 ConversationEntryPointNode_7 {
  id=#1
  class=ConversationEntryPointNode
  outputs [#2]
  properties {entryTag {tagName=Conversation.Entry.Brat}}
 }
 Speech_C_4 {
  id=#2
  class=Speech_C
  outputs [
   #3
   #4
   #5
   #6
   #7
  ]
  properties {
   nodeName=""
   generalConfig {
    defaultMessage="Tasty Bois for life! No expiration date on us."
    contextualMessages [
     {
      message="The bars are raw and the dog is hot. The Brat, at your service."
      contextRequirements [{requirement {class=FortControllerRequirement_IsFirstEverConversationWithNPC}}]
      priority {curve {rowName=Tandem.Speech.Priority.FirstEverInteraction}}
     }
     {
      message="Tasty Bois for life! No expiration date on us."
      contextRequirements [{requirement {class=FortControllerRequirement_CheckNPCMatchConversationCount, properties {nPCUniqueIdTag {tagName=AISpawnerData.Type.Tandem.Brat}}}}]
      priority {curve {rowName=None}}
     }
     {
      message="You might say I'm cooked, but that just means I'm getting tastier."
      contextRequirements [{requirement {class=FortControllerRequirement_CheckNPCMatchConversationCount, properties {nPCUniqueIdTag {tagName=AISpawnerData.Type.Tandem.Brat}}}}]
      priority {curve {rowName=None}}
     }
     {
      message="Who's the king with all the toppings? Ya boi The Brat!"
      contextRequirements [{requirement {class=FortControllerRequirement_CheckNPCMatchConversationCount, properties {nPCUniqueIdTag {tagName=AISpawnerData.Type.Tandem.Brat}}}}]
      priority {curve {rowName=None}}
     }
     {
      message="Turn up the heat. It's time to get TASTY!"
      contextRequirements [{requirement {class=FortControllerRequirement_CheckNPCMatchConversationCount, properties {nPCUniqueIdTag {tagName=AISpawnerData.Type.Tandem.Brat}}}}]
      priority {curve {rowName=None}}
     }
     {
      message="What's tasty, boi??"
      contextRequirements [{requirement {class=FortControllerRequirement_HasCID, properties {allowedCIds [/BRCosmetics/Athena/Items/Cosmetics/Characters/Character_WartyBrine_Sus]}}}]
      priority {curve {rowName=None}}
     }
     {
      message="You think you're still hot, but you smell like leftovers to me."
      contextRequirements [{requirement {
         class=FortControllerRequirement_HasCID
         properties {allowedCIds [
           /BRCosmetics/Athena/Items/Cosmetics/Characters/CID_233_Athena_Commando_M_FortniteDJ
           /BRCosmetics/Athena/Items/Cosmetics/Characters/CID_308_Athena_Commando_F_FortniteDJ
          ]}
        }}]
      priority {curve {rowName=None}}
     }
    ]
   }
  }
 }
 ConversationLinkNode_4 {
  id=#7
  class=ConversationLinkNode
  properties {remoteEntryTag {tagName=Conversation.Entry.Season36.DJYonder.CharacterStories.01}}
 }
 ConversationLinkNode_3 {
  id=#6
  class=ConversationLinkNode
  properties {remoteEntryTag {tagName=Conversation.Entry.Season36.BigDill.CharacterStories.01}}
 }
 ConversationLinkNode_2 {id=#5, class=ConversationLinkNode, properties {remoteEntryTag {tagName=Conversation.Entry.Brat.Dialogue}}}
 ConversationLinkNode_0 {id=#4, class=ConversationLinkNode, properties {remoteEntryTag {tagName=Conversation.Entry.BasicServices}}}
 ConversationLinkNode_1 {id=#3, class=ConversationLinkNode, properties {remoteEntryTag {tagName=Conversation.Entry.NPCLibrary.Quests}}}
]
